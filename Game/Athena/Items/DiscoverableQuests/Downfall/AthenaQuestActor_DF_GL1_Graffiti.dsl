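; Discoverable quest actor for the Downfall GL1 graffiti objective, rewarding players who touch its capsule.
; Once playlist data loads, the server binds overlap, while clients check quest, party-assist, creative or test eligibility and watch quest updates.
; Overlap sends a stat event for the hardcoded quest if incomplete, then pulses PlaySuccessFX, whose OnRep plays sound and Niagara.
; Client objective progress fires empty child hooks and delayed generic success effects.
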(event EventBeginPlay
  ;[1 0.4483 0.1218] STILL TEMP!!!
  (Utilities|FlowControl|Delay 5.0)
  ;[1 1 1] CLIENT AND SERVER - Check if In Creative Mode or not
  (bind _asfort_game_state_athena (Utilities|Casting|CastToFortGameStateAthena (Game|GetGameState)))
  (if (Variables|/Script/FortniteGame.FortGameStateAthena|GetbPlaylistDataIsLoaded _asfort_game_state_athena)
    ;[1 1 1] Check for Quest and Objective
    (Networking|SwitchHasAuthority
      (:Authority
        ;[1 1 1] Bind Overlap Event
        (Collision|BindEventtoOnComponentBeginOverlap (Variables|Default|GetTouchCapsule) (EventDispatchers|CreateEvent :!event Custom|TouchedQuestActor)))
      (:Remote
        (bind _shouldshowlocally (CallFunction|CheckShouldShowLocally))
        (if _shouldshowlocally
          (Variables|Default|SetIsLocallyValid true)
          ;[1 1 1] Bind Events for Touch and QuestUpdated
          (Quests|BindEventtoOnQuestsUpdated (Quests|GetQuestManager (Utilities|Array|Get(acopy) (Fortnite|GetLocalFortPlayerControllers true)) "Invalid") (EventDispatchers|CreateEvent :!event Custom|QuestSuccessfullyUpdated))
          ;;[1 1 1] Tell Children of
          (CallFunction|QuestInitializedandValid))))
    (else
      (Playlist|BindEventtoOnCurrentPlaylistLoaded _asfort_game_state_athena (delegate Custom|PlaylistLoaded)))))

(event Custom|QuestInitializedAndValid)

(event Custom|PlaylistLoaded
  (params (PlaylistName :type Name) (PlaylistContextTags :type Gameplay%20Tag%20Container%20Structure%20%28by%20ref%29))
  ;[1 1 1] Check for Quest and Objective
  (Networking|SwitchHasAuthority
    (:Authority
      ;[1 1 1] Bind Overlap Event
      (Collision|BindEventtoOnComponentBeginOverlap (Variables|Default|GetTouchCapsule) (EventDispatchers|CreateEvent :!event Custom|TouchedQuestActor)))
    (:Remote
      (bind _shouldshowlocally (CallFunction|CheckShouldShowLocally))
      (if _shouldshowlocally
        (Variables|Default|SetIsLocallyValid true)
        ;[1 1 1] Bind Events for Touch and QuestUpdated
        (Quests|BindEventtoOnQuestsUpdated (Quests|GetQuestManager (Utilities|Array|Get(acopy) (Fortnite|GetLocalFortPlayerControllers true)) "Invalid") (EventDispatchers|CreateEvent :!event Custom|QuestSuccessfullyUpdated))
        ;;[1 1 1] Tell Children of
        (CallFunction|QuestInitializedandValid)))))

(event Custom|TouchedQuestActor
  (params (OverlappedComponent :type Primitive%20Component%20Object%20Reference) (OtherActor :type Actor%20Object%20Reference) (OtherComp :type Primitive%20Component%20Object%20Reference) (OtherBodyIndex :type Integer) (bFromSweep :type Boolean) (SweepResult :type Hit%20Result%20Structure%20%28by%20ref%29))
  (Development|PrintString "Hello")
  ;[1 1 1] Touched Capsule
  (CallFunction|FoundEvent (Fortnite|GetFortPlayerControllerfromActor OtherActor)))

(event Custom|FoundEvent
  (params (PC :type Fort%20Player%20Controller%20Object%20Reference))
  ;[1 1 1] Touched Capsule
  (bind _asfort_player_controller_athena (Utilities|Casting|CastToFortPlayerControllerAthena PC))
  (bind _playerhasquest (Class|AthenaQuestFunctionLibrary|AthenaHasQuest PC "/Game/Athena/Items/Quests/DiscoverableQuests/Downfall_Quests/AthenaQuest_DF_GL1.AthenaQuest_DF_GL1"))
  (if _playerhasquest
    (bind (_playercontrollerout _completedquestobjective) (Quest|HasPlayerCompletedQuestObjective PC "/Game/Athena/Items/Quests/DiscoverableQuests/Downfall_Quests/AthenaQuest_DF_GL1.AthenaQuest_DF_GL1" "quest_df_gl1_obj01"))
    (if (not _completedquestobjective)
      (Quest|SendComplexCustomStatEvent (Quests|GetQuestManager _asfort_player_controller_athena "Invalid") PC "(GameplayTags=)" "(GameplayTags=((TagName=\"Athena.Quests.Narrative.Downfall.Graffiti\")))" 1)
      (Variables|Default|SetPlaySuccessFX true)
      (Utilities|FlowControl|Delay 1.0)
      (Variables|Default|SetPlaySuccessFX false))))

(event Custom|QuestSuccessfullyUpdated
  ;[1 1 1] Quest Updated
  (if (Variables|Default|GetisLocallyValid)
    (bind _questcountachieved (Class|AthenaQuestFunctionLibrary|AthenaQuestObjectiveCounter (Utilities|Array|Get(acopy) (Fortnite|GetLocalFortPlayerControllers true)) (Variables|Default|GetQuestItem) (Variables|Default|GetObjectiveStat)))
    (if (> _questcountachieved (Variables|Default|GetPlayerObjectiveCount))
      (Variables|Default|SetIsLocallyValid false)
      (CallFunction|ObjectiveSuccessfullyCompleted (Variables|Default|GetPlayerThatTriggeredObject)))))

(event Custom|ObjectiveSuccessfullyCompleted
  (params (PlayerWhoTriggeredObject :type Actor%20Object%20Reference))
  (bind _genericsuccesssound (Variables|Default|GetGenericSuccessSound))
  (bind _genericsuccessfx (Variables|Default|GetGenericSuccessFX))
  ;[1 1 1] Success FX and Sound
  (CallFunction|TellChildOfSuccess)
  (Utilities|FlowControl|Delay (Variables|Default|GetDelayBeforeGenericSuccessSoundAndFX))
  (Utilities|IsValid _genericsuccesssound
    (:Is%20Valid
      (Audio|SpawnSoundatLocation _genericsuccesssound (Transformation|GetActorLocation)))
    (:Is%20Not%20Valid
      _))
  (Utilities|IsValid _genericsuccessfx
    (:Is%20Valid
      (Effects|SpawnFXSystemAttached _genericsuccessfx (Variables|/Script/Engine.Character|GetMesh (Game|Player|GetPlayerPawn (Utilities|Array|Get(acopy) (Fortnite|GetLocalFortPlayerControllers true))))))))

(event Custom|TellChildOfSuccess)

(event Custom|TouchedActorInCreativeMode
  ;[1 1 1] Success FX and Sound
  (Networking|SwitchHasAuthority
    (:Remote
      (CallFunction|TellChildOfSuccess))))

(fn CheckShouldShowLocally
  (bind _returnvalue (Fortnite|GetLocalFortPlayerControllers true))
  (bind _output (Utilities|Array|Get(acopy) _returnvalue))
  (bind _objbackendname (Variables|Default|GetObjBackendName))
  (bind (_questisvalid _questcompleted _objectivecompleted) (Default|AthenaCheckQuestandObjectiveCompletion _output (Variables|Default|GetQuestItem) _objbackendname))
  (bind (_bisassistingobjective _bisvalidplayercontroller) (Class|AthenaQuestFunctionLibrary|AthenaPartyAssistCheck _output _objbackendname))
  (Variables|Default|SetbIsQuestShared (and _bisassistingobjective _bisvalidplayercontroller))
  (return (Math|Boolean|ORBoolean (and _questisvalid (not _objectivecompleted)) (Variables|Default|GetCreativeMode) (Variables|Default|GetbTestQuest) (Variables|Default|GetbIsQuestShared))))

(fn OnRep_PlaySuccessFX
  (if (Variables|Default|GetPlaySuccessFX)
    (Audio|Components|Audio|Play (Variables|Default|GetAthena_CollectHiddenStarReward_Cue) 0.0)
    (Components|Activation|Activate (Variables|Default|GetNS_Athena_QuestItem_PickupSuccess) false)))

(fn OnRep_ShoppingCartThatTriggered)

(fn ConstructionScript)
